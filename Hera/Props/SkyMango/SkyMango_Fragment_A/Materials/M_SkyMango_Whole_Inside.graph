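# Material for the inside of the SkyMango prop: a glowing, iridescent crystal-like surface with animated crevices.
# The deferred branch layers thin-film color, parallax sparkles, scratch detail normals and tinted emissive, with ray-tracing switches for Lumen levels.
# A shading-path switch sends forward and mobile to a simpler branch with sphere-masked emissive shading.
# MF_SkyMango_Whole_WPO drives vertex animation and rotated world-space normals, and supplies animation curves that modulate the glow.
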Material /Hera/Props/SkyMango/SkyMango_Fragment_A/Materials/M_SkyMango_Whole_Inside
graph {
 properties {
  bAlwaysEvaluateWorldPositionOffset=true
  bAutomaticallySetUsageInEditor=false
  bTangentSpaceNormal=false
  bUseEmissiveForDynamicAreaLighting=true
  bUseMaterialAttributes=true
  pixelDepthOffsetMode=PDOM_Legacy
  previewMesh=/Hera/Props/SkyMango/SkyMango_Fragment_A/Meshes/SM_SkyMango_Whole
  usedShadingModels=Default Lit
 }
 nodes [
  n1 = NamedRerouteUsage()
  n2 = NamedRerouteUsage()
  n3 = MaterialFunctionCall() {materialFunction=&ObjectScale}
  n4 = SphereMask(
   A=WorldPosition() {worldPositionShaderOffset=WPT_ExcludeAllShaderOffsets}.XYZ
   B=Constant3Vector()
   Radius=@n3.Scale X
   Hardness=ScalarParameter() {defaultValue=1, group=EmissiveColors, parameterName=Hardness}
  )
  n5 = StaticSwitchParameter(True=@n4, False=OneMinus(None=@n4)) {group=EmissiveColors, parameterName=UseWorldPosition}
  n6 = Add(A=Multiply(A=Add(A=@n1) {constB=0.1}, B=Add(A=@n2) {constB=0.1}), B=Multiply(A=@n5) {constB=0.2})
  n7 = Sign(None=ScalarParameter() {group=EmissiveColors, parameterName=IsZeroPointShard? Switch})
  LinearInterpolate(
   A=@n5
   B=Saturate(None=Multiply(A=@n6) {constB=2.5})
   Alpha=NamedRerouteDeclaration(None=Saturate(None=@n7)) {name=Switch, nodeColor=<rgba 0.3828 1 0 1>}
  )
  Max(B=ScalarParameter() {defaultValue=1, parameterName=LumenClamp})
  n8 = DotProduct(A=CameraVectorWS(), B=VertexNormalWS())
  n9 = Time()
  n10 = Sine(None=@n9) {period=3.7}
  n11 = Sine(None=@n9) {period=1.6}
  Multiply(
   A=Power(Base=Saturate(None=@n8), Exp=ScalarParameter() {defaultValue=8, group=Crevice, parameterName=Crevice_Fresnel_Pow})
   B=Max(A=ConstantBiasScale(None=@n10), B=ConstantBiasScale(None=@n11))
  )
  n12 = TextureSampleParameter2D() {
   group=Textures
   parameterName=SpecularMasks
   samplerType=SAMPLERTYPE_Masks
   texture=/Hera/Props/SkyMango/SkyMango_Fragment_A/Textures/T_SkyMango_Fragment_Inner_S2shards_M
  }
  NamedRerouteDeclaration(None=@n12.R) {name=Curvature, nodeColor=<rgba 1 0.019 0 1>}
  NamedRerouteDeclaration(None=@n12.G) {name=Cavity, nodeColor=<rgba 0.0265 1 0.1272 1>}
  n13 = ScalarParameter() {defaultValue=1505.5891, parameterName=Mobile shadow Rad}
  n14 = Divide(A=ObjectRadius(), B=ScalarParameter() {defaultValue=512, group=RVT, parameterName=OR Mobile}) {constB=512}
  n15 = LinearInterpolate(
   B=ScalarParameter() {defaultValue=0.2757, parameterName=MobileshadowCal}
   Alpha=SphereMask(A=WorldPosition().XYZ, B=ActorPositionWS(), Radius=Multiply(A=@n13, B=@n14), Hardness=Constant() {r=0.2})
  ) {constA=1}
  n16 = WorldPosition()
  n17 = VectorParameter() {parameterName=Zero, defaultValue=<rgba 7400 5800 0 0>}
  n18 = ScalarParameter() {defaultValue=7500, parameterName=MobileRadius}
  n19 = ScalarParameter() {defaultValue=0.5, parameterName=MobileHardness}
  n20 = NamedRerouteUsage()
  n21 = NamedRerouteUsage()
  n22 = LinearInterpolate(A=Constant() {r=0.3}, Alpha=NamedRerouteUsage()) {constB=0}
  n23 = NamedRerouteUsage()
  n24 = NamedRerouteUsage()
  n25 = NamedRerouteUsage()
  n26 = SetMaterialAttributes!(
   MaterialAttributes
   Base Color=Multiply(A=ScalarParameter() {defaultValue=9.3231, parameterName=DiffuseLevel Mobile}, B=Multiply(A=@n20, B=@n21))
   Emissive Color=Add(A=Multiply(A=@n22, B=@n23), B=Multiply(A=@n24, B=@n25))
   Specular=ScalarParameter() {defaultValue=0.1, group=Textures, parameterName=Specular}
   Roughness=ScalarParameter() {defaultValue=0.4, group=Textures, parameterName=Roughness}
   Normal=NamedRerouteUsage()
   Opacity Mask=NamedRerouteUsage()
  ) {attributeSetTypes [
    #1
    #2
    #3
    #4
    #5
    #6
   ]}
  n27 = GetMaterialAttributes(None=@n26) {attributeGetTypes [#2]}
  n28 = Multiply(
   A=LinearInterpolate(
    A=VectorParameter() {parameterName=ColorCenterMobile, defaultValue=<rgba 0.4531 0.4531 0.4531 0>}.RGB
    B=VectorParameter() {parameterName=ColorEyeMobile, defaultValue=<rgba 0.1536 0.1536 0.1536 0>}.RGB
    Alpha=SphereMask(A=@n16.XYZ, B=@n17.RGB, Radius=@n18, Hardness=@n19)
   )
   B=@n27.Emissive Color
  )
  n29 = SetMaterialAttributes!(
   MaterialAttributes=@n27.MaterialAttributes
   Emissive Color=Multiply(A=Multiply(A=@n15, B=@n28), B=ScalarParameter() {defaultValue=1, parameterName=Emissive Intensity Mobile})
  ) {attributeSetTypes [#2]}
  n30 = OneMinus(None=Fresnel() {baseReflectFraction=0.3, exponent=3})
  n31 = Multiply(A=Multiply(A=NamedRerouteUsage(), B=ScalarParameter() {defaultValue=0.1, parameterName=ThinFilm}), B=@n30)
  n32 = Multiply(A=Multiply(A=@n30, B=NamedRerouteUsage()), B=NamedRerouteUsage())
  n33 = NamedRerouteUsage()
  n34 = Multiply(
   A=TextureCoordinate()
   B=Divide(A=ObjectRadius(), B=ScalarParameter() {defaultValue=2000, group=Details, parameterName=ObjectRadius}) {constB=512}
  )
  n35 = TextureSample(UVs=Multiply(A=@n34, B=ScalarParameter() {group=Details, parameterName=ScaleDetail})) {samplerType=SAMPLERTYPE_Normal, texture=/Game/Global/Textures/Default/Tileable_Patterns/T_IceScratch1_N}
  n36 = Constant3Vector() {constant=<rgba 0.25 0.25 0 0>}
  n37 = Add(A=NamedRerouteUsage(), B=Multiply(A=OneMinus(None=@n33), B=Multiply(A=@n35.RGB, B=@n36)))
  n38 = NamedRerouteUsage()
  n39 = Multiply(A=ScalarParameter() {defaultValue=25, group=Details, parameterName=Sparkles}, B=@n34)
  n40 = Multiply(A=NamedRerouteUsage()) {constB=-1}
  n41 = Multiply(
   A=OneMinus(None=NamedRerouteUsage())
   B=TextureSample(UVs=BumpOffset(Coordinate=@n39, Height=@n40)) {samplerType=SAMPLERTYPE_Masks, texture=/Game/Environments/Apollo/Sets/DesertPoint/Impact/NewSand_M}.G
  )
  n42 = Add(A=Multiply(A=@n31) {constB=0.5}, B=Multiply(A=Multiply(A=@n38, B=@n41), B=NamedRerouteUsage()))
  n43 = NamedRerouteUsage()
  n44 = Constant() {r=3}
  n45 = Multiply(A=Add(A=Power(Base=@n43, Exp=@n44), B=NamedRerouteUsage()), B=NamedRerouteUsage())
  n46 = GetMaterialAttributes(None=MakeMaterialAttributes!(
    BaseColor=Add(A=@n31, B=@n32)
    Metallic=Constant() {r=0.5238}
    Specular=ScalarParameter() {defaultValue=0.1, group=Textures, parameterName=Specular}
    Roughness=ScalarParameter() {defaultValue=0.4, group=Textures, parameterName=Roughness}
    Anisotropy
    EmissiveColor=Add(A=@n42, B=@n45)
    Opacity
    OpacityMask=NamedRerouteUsage()
    Normal=Normalize(VectorInput=@n37)
    Tangent
    WorldPositionOffset
    SubsurfaceColor
    ClearCoat
    ClearCoatRoughness
    AmbientOcclusion
    Refraction
    CustomizedUVs_0
    CustomizedUVs_1
    CustomizedUVs_2
    CustomizedUVs_3
    CustomizedUVs_4
    CustomizedUVs_5
    CustomizedUVs_6
    CustomizedUVs_7
    PixelDepthOffset
    ShadingModel
    Displacement
   )) {attributeGetTypes [#2]}
  n47 = VectorParameter() {parameterName=EmissiveTint, defaultValue=<rgba 0.55 0.775 1 1>}
  n48 = NamedRerouteUsage()
  n49 = ComponentMask(None=NamedRerouteDeclaration(None=MaterialFunctionCall!(Tangent Normal, UseWobble) {materialFunction=/Game/Environments/Hera/FX/ZeroPoint/Orb/Materials/MF_SkyMango_Whole_WPO}.Curve) {name=AnimCurves, nodeColor=<rgba 0 1 0.9297 1>}) {b=true}
  n50 = RayTracingQualitySwitch(Normal=@n49, RayTraced=Min(A=@n49, B=ScalarParameter() {defaultValue=1, parameterName=LumenClamp}))
  n51 = Multiply(
   A=Add(A=Multiply(A=@n46.Emissive Color, B=@n47.RGB), B=Multiply(A=@n48, B=@n50))
   B=ScalarParameter() {defaultValue=1.5, group=EmissiveColors, parameterName=Emissive Intensity}
  )
  n52 = ScalarParameter() {defaultValue=1, group=EmissiveColors, parameterName=LumenEmissiveMultiplier}
  n53 = SetMaterialAttributes(MaterialAttributes=@n46.MaterialAttributes, Emissive Color=RayTracingQualitySwitch(Normal=@n51, RayTraced=Multiply(A=@n51, B=@n52))) {attributeSetTypes [#2]}
  n54 = GetMaterialAttributes(None=ShadingPathSwitch(Default=@n53, Deferred=@n53, Forward=@n29, Mobile=@n29)) {attributeGetTypes [#5]}
  n55 = MaterialFunctionCall(Tangent Normal=@n54.Normal) {materialFunction=&MF_SkyMango_Whole_WPO}
  NamedRerouteDeclaration(None=@n55.WorldNormal) {name=Rotating Normals WS, nodeColor=<rgba 0 1 0.1562 1>}
  n56 = Multiply(A=TextureCoordinate(), B=ScalarParameter() {defaultValue=2.7, group=Crevice, parameterName=Crevice_PanTiling})
  n57 = Frac(None=Multiply(A=Time(), B=ScalarParameter() {defaultValue=0.12, group=Crevice, parameterName=Crevice_PanSpeed}))
  n58 = TextureCoordinate()
  n59 = ScalarParameter() {defaultValue=2, group=Crevice, parameterName=Crevice_PanTilingDistort}
  n60 = Multiply(A=Time(), B=ScalarParameter() {defaultValue=-0.27, group=Crevice, parameterName=Crevice_PanSpeedDistort})
  n61 = TextureSampleParameter2D(UVs=Add(A=Multiply(A=@n58, B=@n59), B=Frac(None=@n60))) {
   group=Crevice
   parameterName=Crevice_PanDistortTexture
   samplerSource=SSM_Wrap_WorldGroupSettings
   texture=/Game/TimeOfDay/LightFunction/T_CloudNoise_Artemis
  }
  n62 = ConstantBiasScale(None=@n61.R) {bias=-0.5, scale=2}
  n63 = ScalarParameter() {defaultValue=0.0431, group=Crevice, parameterName=Crevice_Pan_DistortIntensity}
  n64 = TextureSampleParameter2D(UVs=Add(A=Add(A=@n56, B=@n57), B=Multiply(A=@n62, B=@n63))) {
   group=Crevice
   parameterName=Crevice_PanTexture
   samplerSource=SSM_Wrap_WorldGroupSettings
   texture=/Game/Effects/Fort_Effects/Textures/Plasma/T_PlasmaNoise_A_01
  }
  n65 = ScalarParameter() {group=Crevice, parameterName=Crevice_RGBPan}
  n66 = NamedRerouteUsage()
  n67 = Multiply(
   A=Power(Base=OneMinus(None=@n66), Exp=ScalarParameter() {defaultValue=4, group=Crevice, parameterName=CrevicePow})
   B=VectorParameter() {group=Crevice, parameterName=CreviceEmissiveColor, defaultValue=<rgba 29.5652 0.5913 5.4203 1>}.RGB
  )
  n68 = LinearInterpolate(
   A=ScalarParameter() {defaultValue=2, parameterName=Crevice Pop Mult A}
   B=ScalarParameter() {defaultValue=1, parameterName=Crevice Pop Mult B}
   Alpha=ComponentMask(None=NamedRerouteUsage()) {r=true}
  )
  NamedRerouteDeclaration(None=Multiply(A=LinearInterpolate(A=@n64.R, B=@n64.R, Alpha=@n65), B=Multiply(A=@n67, B=@n68))) {name=Crevice, nodeColor=<rgba 0 0.1094 1 1>}
  n69 = TextureSampleParameter2D() {
   group=Textures
   parameterName=Normals
   samplerType=SAMPLERTYPE_Normal
   texture=/Hera/Props/SkyMango/SkyMango_Fragment_A/Textures/T_SkyMango_Fragment_Inner_S2shards_N
  }
  NamedRerouteDeclaration(None=@n69.RGB) {name=Tangent Normal, nodeColor=<rgba 0 0.2 1 1>}
  n70 = CameraVectorWS()
  n71 = PixelNormalWS()
  NamedRerouteDeclaration(None=TextureSample(UVs=DotProduct(A=@n70, B=@n71)) {samplerSource=SSM_Clamp_WorldGroupSettings, texture=/Game/Weapons/WeaponSkins/Wraps/Textures/T_Grad_Rainbow2}.RGB) {name=Iridescence, nodeColor=<rgba 1 0 0.8437 1>}
  NamedRerouteDeclaration(None=LinearInterpolate(
    A=VectorParameter() {group=DiffuseColors, parameterName=Color, defaultValue=<rgba 0.15 0 0.15 1>}.RGB
    B=VectorParameter() {group=DiffuseColors, parameterName=Color2, defaultValue=<rgba 0.2 0.012 0.0924 1>}.RGB
    Alpha=NamedRerouteUsage()
   )) {name=Color Base, nodeColor=<rgba 1 0.2344 0 1>}
  NamedRerouteDeclaration(None=LinearInterpolate(
    A=VectorParameter() {group=EmissiveColors, parameterName=EcolorMobile, defaultValue=<rgba 0.5 0.03 0.1867 1>}.RGB
    B=VectorParameter() {group=EmissiveColors, parameterName=Ecolor2Mobile, defaultValue=<rgba 4 0.2083 0.8403 1>}.RGB
    Alpha=NamedRerouteUsage()
   )) {name=Colors Mobile, nodeColor=<rgba 0.9378 0.6616 1 1>}
  NamedRerouteDeclaration(None=LinearInterpolate(
    A=VectorParameter() {group=EmissiveColors, parameterName=Ecolor, defaultValue=<rgba 15 1.2 2.3501 1>}.RGB
    B=VectorParameter() {group=EmissiveColors, parameterName=Ecolor2, defaultValue=<rgba 1 0.06 0.4517 1>}.RGB
    Alpha=NamedRerouteUsage()
   )) {name=Ecolors Defered, nodeColor=<rgba 0.4062 1 0 1>}
  NamedRerouteDeclaration(None=@n12.B) {name=Opacity Mask Texture, nodeColor=<rgba 0 0.8125 1 1>}
  n72 = Length(None=LocalPosition() {includedOffsets=ExcludeOffsets}.XYZ) {is=class MaterialExpressionLength}
  n73 = ComponentMask(None=ObjectLocalBounds().Half Extents) {r=true}
  NamedRerouteDeclaration(None=SmoothStep(
    Min=ScalarParameter() {defaultValue=0.4, parameterName=Sphere Mask Inner}
    Max=ScalarParameter() {defaultValue=0.7734, parameterName=Sphere Mask Outer}
    Value=Divide(A=@n72, B=@n73)
   )) {name=Sphere Mask, nodeColor=<rgba 0.6875 0 1 1>}
  n74 = DotProduct(A=CameraVectorWS(), B=Transform(None=NamedRerouteUsage()))
  n75 = ScalarParameter() {defaultValue=0.5, group=EmissiveColors, parameterName=Fresnel Power}
  NamedRerouteDeclaration(None=Saturate(None=Power(Base=@n74, Exp=@n75))) {name=Fresnel, nodeColor=<rgba 0 1 0.6016 1>}
  NamedRerouteDeclaration(None=DotProduct(A=NamedRerouteUsage(), B=ReflectionVectorWS())) {name=Reflection Vector, nodeColor=<rgba 0.7109 1 0 1>}
  NamedRerouteUsage()
  SetMaterialAttributes!(MaterialAttributes=@n54.MaterialAttributes, World Position Offset=@n55.WPO, Normal=@n55.WPO) {attributeSetTypes [#7, #5]}
  n76 = VectorParameter() {group=EmissiveColors, parameterName=SphereMaskAdjust}
  StaticSwitchParameter(
   True=VectorParameter() {parameterName=Sphere, defaultValue=<rgba 7000 5000 0 1>}.RGB
   False=Add(A=ObjectPositionWS(), B=Transform(None=@n76.RGB) {transformSourceType=TRANSFORMSOURCE_Local})
  ) {group=EmissiveColors, parameterName=UseWorldPosition}
  n77 = ScalarParameter() {group=EmissiveColors, parameterName=Radius}
  StaticSwitchParameter(True=@n77, False=Multiply(A=@n77, B=@n3.Scale XYZ)) {group=EmissiveColors, parameterName=UseWorldPosition}
 ]
}
