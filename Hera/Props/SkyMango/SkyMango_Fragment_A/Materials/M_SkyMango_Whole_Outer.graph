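# Glowing, cracked-glass material for the outer shell of the whole SkyMango prop.
# Crack masks sampled at two tilings drive pulsing glow, clamped for ray tracing, plus flickering inner and outer Fresnel, dimmed by the Nexus core mask.
# MF_SkyMango_Whole_WPO supplies world-space normals, WPO and crack-animation curves; an older bob-and-rotate WPO remains under "OLD WPO".
# MaterialProxyReplace puts the crack glow into base color for proxy bakes.
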Material /Hera/Props/SkyMango/SkyMango_Fragment_A/Materials/M_SkyMango_Whole_Outer
graph {
 properties {
  bAlwaysEvaluateWorldPositionOffset=true
  bAutomaticallySetUsageInEditor=false
  bTangentSpaceNormal=false
  bUseEmissiveForDynamicAreaLighting=true
  usedShadingModels=Default Lit
 }
 nodes [
  Constant() {r=1}
  Constant() {r=1}
  n1 = Multiply(A=TextureCoordinate() {uTiling=2, vTiling=2}, B=NamedRerouteUsage())
  n2 = Multiply(A=NamedRerouteUsage(), B=TextureCoordinate())
  n3 = MaterialFunctionCall(
   Normal=LinearInterpolate(
    A=TextureSample(UVs=@n1) {samplerType=SAMPLERTYPE_Normal, texture=/Hera/Props/SkyMango/SkyMango_Fragment_A/Textures/T_Glass_Cracks_N2}.RGB
    B=TextureSample(UVs=@n2) {samplerType=SAMPLERTYPE_Normal, texture=&T_Glass_Cracks_N2}.RGB
   )
   Flatness=ScalarParameter() {defaultValue=-3, parameterName=Normal Flatness}
  ) {materialFunction=&FlattenNormal}
  n4 = Add(
   A=Multiply(A=Time(), B=ScalarParameter() {defaultValue=0.025, group=WPO, parameterName=Rotation Speed})
   B=ScalarParameter() {defaultValue=-0.01, group=WPO, parameterName=Rotation Angle Min}
  )
  LinearInterpolate(
   A=@n3.Result
   B=MaterialFunctionCall(Rotation Axis=Constant3Vector() {constant=<rgba 0 0 1 0>}, Rotation Angle=@n4, World Space Vertex Normals=Transform(None=@n3.Result)) {materialFunction=&FixRotateAboutAxisNormals}.New Tangent Space Vertex Normal
   Alpha=Saturate(None=ScalarParameter() {defaultValue=1, parameterName="Use Rotation WPO? (Switch)", sliderMax=1})
  )
  MaterialFunctionCall!(UVs=TextureCoordinate() {uTiling=20, vTiling=20}, Tiling) {materialFunction=/Engine/ArtTools/RenderToTexture/MaterialFunctions/CheckerPattern}
  MaterialProxyReplace(Realtime=Constant(), MaterialProxy=Constant() {r=0.9})
  MaterialProxyReplace(Realtime=ScalarParameter() {defaultValue=0.05, parameterName=Roughness}, MaterialProxy=Constant() {r=0.4})
  MaterialProxyReplace(Realtime=Constant(), MaterialProxy=Constant() {r=0.9})
  MaterialProxyReplace(Realtime=ScalarParameter() {defaultValue=0.05, parameterName=Roughness}, MaterialProxy=Constant() {r=0.4})
  n5 = VectorParameter() {parameterName=CrackTiling, defaultValue=<rgba 1.5 1.5 0 1>}
  NamedRerouteDeclaration(None=AppendVector(A=@n5.R, B=@n5.R)) {name=CrackTiling, nodeColor=<rgba 0 0.4844 1 1>}
  n6 = VectorParameter() {parameterName=EmissiveTiling, defaultValue=<rgba 1.5 1.5 0 0>}
  NamedRerouteDeclaration(None=AppendVector(A=@n6.R, B=@n6.R)) {name=EmissiveTiling, nodeColor=<rgba 0 0.6719 1 1>}
  NamedRerouteDeclaration(None=@n3.Result) {name=Tangent Normal, nodeColor=<rgba 0.6406 0 1 1>}
  n7 = Time()
  n8 = ScalarParameter() {defaultValue=0.4, group=WPO, parameterName=Bob Sine Speed}
  n9 = LinearInterpolate(
   A=VectorParameter() {group=WPO, parameterName=BobHeightMax, defaultValue=<rgba 0 0 4 0>}.RGB
   B=VectorParameter() {group=WPO, parameterName=BobHeightMin, defaultValue=<rgba 0 0 -4 0>}.RGB
   Alpha=Sine(None=Multiply(A=@n7, B=@n8))
  )
  n10 = RotateAboutAxis(
   NormalizedRotationAxis=Constant3Vector() {constant=<rgba 0 0 1 0>}
   RotationAngle=@n4
   PivotPoint=ObjectPositionWS()
   Position=WorldPosition() {worldPositionShaderOffset=WPT_ExcludeAllShaderOffsets}.XYZ
  )
  n11 = ScalarParameter() {defaultValue=1, parameterName="Use Rotation WPO? (Switch)", sliderMax=1}
  NamedRerouteDeclaration(None=LinearInterpolate(B=Add(A=@n9, B=@n10), Alpha=Saturate(None=@n11))) {name=OLD WPO, nodeColor=<rgba 1 0 0.8203 1>}
  n12 = MaterialFunctionCall!(Tangent Normal, UseWobble) {materialFunction=/Game/Environments/Hera/FX/ZeroPoint/Orb/Materials/MF_SkyMango_Whole_WPO}
  n13 = OneMinus(None=ComponentMask(None=@n12.Curve) {r=true})
  n14 = LinearInterpolate(
   A=ScalarParameter() {defaultValue=0.5, parameterName=Crack WPO Anim Brightness A}
   B=ScalarParameter() {defaultValue=1.5, parameterName=Crack WPO Anim Brightness B}
   Alpha=@n12.Wobble
  ) {constA=1}
  n15 = Multiply(A=Time(), B=ScalarParameter() {defaultValue=0.33, parameterName=Crack Pulse})
  n16 = Add(
   A=Add(A=ComponentMask(None=@n12.Curve) {b=true}, B=Multiply(A=@n13, B=@n14))
   B=LinearInterpolate(A=ScalarParameter() {defaultValue=0.02, parameterName=Crack Pulse Darkening}, Alpha=Sine(None=@n15)) {constA=0.4}
  )
  n17 = Min(A=@n16, B=ScalarParameter() {defaultValue=1, parameterName=LumenClamp})
  n18 = NamedRerouteUsage()
  n19 = TextureSample(UVs=Multiply(A=TextureCoordinate() {uTiling=2, vTiling=2}, B=@n18)) {texture=/Hera/Props/SkyMango/SkyMango_Fragment_A/Textures/T_Glass_Cracks_M}
  n20 = TextureSample(UVs=Multiply(A=@n18, B=TextureCoordinate())) {texture=&T_Glass_Cracks_M}
  n21 = Multiply(
   A=Multiply(
    A=RayTracingQualitySwitch(Normal=@n16, RayTraced=@n17)
    B=VectorParameter() {parameterName=CrackGlow, defaultValue=<rgba 2.3957 20.6478 50 1>}.RGB
   )
   B=Power(Base=LinearInterpolate(A=@n19.G, B=@n20.G) {constAlpha=0.7})
  )
  n22 = ScalarParameter() {defaultValue=5, parameterName=FlickerSpeed}
  n23 = CollectionParameter() {collection=/Game/Athena/Environments/Nexus/Materials/FXshader/MPC_Nexus, parameterName=Destabilize}
  n24 = ScalarParameter() {parameterName=Destabilize}
  n25 = Multiply(
   A=Time()
   B=LinearInterpolate(A=@n22, B=Multiply(A=@n22) {constB=2}, Alpha=StaticSwitchParameter(True=@n23, False=@n24) {parameterName=UseMPC})
  )
  n26 = Sine(None=@n25) {period=0.55}
  n27 = Sine(None=@n25) {period=0.11}
  n28 = Add(A=Add(A=Multiply(B=@n26), B=Multiply(A=@n27) {constB=0.2}), B=Multiply(A=Sine(None=@n25) {period=0.175}) {constB=0.3})
  n29 = ConstantBiasScale(None=Multiply(A=Sine(None=@n25) {period=3.25}, B=Sine(None=@n25) {period=2.325}))
  n30 = Constant() {r=0.6}
  n31 = ShadingPathSwitch(Default=QualitySwitch(Default=Multiply(A=@n28, B=@n29), Low=@n30), Mobile=@n30) {desc=Migrated from FeatureLevelSwitch}
  n32 = Multiply(A=@n31) {constB=0.01}
  n33 = Add(A=ScalarParameter() {defaultValue=8, parameterName=Outer Shell Fresnel}, B=@n32)
  n34 = Fresnel(ExponentIn=Add(A=@n32, B=ScalarParameter() {defaultValue=0.4, parameterName=Inner Glow Fresnel})) {baseReflectFraction=0}
  n35 = ScalarParameter() {defaultValue=8, parameterName=Inner Fresnel Power}
  n36 = Multiply(
   A=Power(Base=OneMinus(None=@n34), Exp=Add(A=@n32, B=@n35))
   B=VectorParameter() {parameterName=Inner Fresnel Color, defaultValue=<rgba 0.8023 1.9068 3 1>}.RGB
  )
  n37 = MaterialFunctionCall!(
   Time Multiplier=CollectionParameter() {collection=/Game/Athena/Environments/Nexus/Materials/NexCollection, parameterName=TimeMultiplier}
   Lerp Shatter Pulse=CollectionParameter() {collection=&NexCollection, parameterName=PulseLerp}
   Shake=CollectionParameter() {collection=&NexCollection, parameterName=Shake}
   Progress=CollectionParameter() {collection=&NexCollection, parameterName=Progress}
  ) {materialFunction=/Game/Athena/Environments/Nexus/Materials/MF_Nex_Core_WPO}
  n38 = Saturate(None=LinearInterpolate(Alpha=@n37.Mask) {constA=1, constB=-0.5})
  n39 = Add(
   A=Multiply(
    A=VectorParameter() {parameterName=Outer Fresnel Color, defaultValue=<rgba 0.25 0.1156 0.2376 1>}.RGB
    B=Fresnel(ExponentIn=@n33) {baseReflectFraction=0}
   )
   B=Multiply(A=LinearInterpolate(Alpha=@n31) {constA=0.7}, B=Multiply(A=@n36, B=@n38))
  )
  NamedRerouteDeclaration(None=MaterialFunctionCall(Emissive=Add(A=@n21, B=@n39)) {materialFunction=&MF_FX_EmissiveAdjustments}.Result) {name=OUT Emissive, nodeColor=<rgba 1 0 0.9609 1>}
  n40 = Constant3Vector() {constant=<rgba 0.004 0.0284 0.04 1>}
  NamedRerouteDeclaration(None=MaterialProxyReplace(Realtime=Constant(), MaterialProxy=Add(A=@n40, B=@n21))) {name=OUT BaseColor, nodeColor=<rgba 0 1 0.7188 1>}
  n41 = MaterialFunctionCall(Tangent Normal=NamedRerouteUsage()) {materialFunction=&MF_SkyMango_Whole_WPO}
  NamedRerouteDeclaration(None=@n41.WorldNormal) {name=OUT Normal, nodeColor=<rgba 1 0 0.8906 1>}
  NamedRerouteDeclaration(None=@n41.WPO) {name=OUT WPO, nodeColor=<rgba 0 1 0.7891 1>}
  NamedRerouteUsage()
  NamedRerouteUsage()
  NamedRerouteUsage()
  NamedRerouteUsage()
  ObjectPositionWS() {originType=CameraRelative}
  n42 = ScalarParameter() {defaultValue=1, parameterName=Use WPO}
  Reroute(None=Add(A=LinearInterpolate(B=@n37.WPOSpread, Alpha=@n42), B=NamedRerouteUsage()))
  TemporalResponsivenessOutput!(None=ScalarParameter() {defaultValue=0.5, parameterName=TempResp})
  TextureSample() {samplerType=SAMPLERTYPE_Normal, texture=/Game/Effects/Fort_Effects/Textures/Tile/T_Glass_Cracks_N}
  TextureSample() {samplerType=SAMPLERTYPE_Normal, texture=&T_Glass_Cracks_N}
  WorldPosition() {worldPositionShaderOffset=WPT_CameraRelativeNoOffsets}
 ]
}
