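# Parent unlit additive material for the SkyMango Fragment A echo, a flickering glassy ghost pulled toward a Custom Location parameter.
# Sine-driven phases build vertex offsets for attraction and vertex-color-masked streaks, with PreviousFrameSwitch faking motion vectors.
# Crack normals and masks drive fresnel opacity and two-color emissive, with looping age fade.
# IsStreaksMesh and IsOuterShell switches vary layers; non-streak meshes also mask opacity against custom depth.
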Material /Hera/Props/SkyMango/SkyMango_Fragment_A/Materials/SkyMango_Fragment_Echo_PARENT
graph {
 properties {
  allowTranslucentCustomDepthWrites=true
  bAutomaticallySetUsageInEditor=false
  bOutputTranslucentVelocity=true
  bUsedWithNanite=true
  blendMode=BLEND_Additive
  opacityMaskClipValue=0
  shadingModel=MSM_Unlit
  translucencyPass=MTP_BeforeDOF
  usedShadingModels=Unlit
 }
 nodes [
  CameraVectorWS()
  Frac(None=Sine(None=Multiply() {constB=0.01}))
  n1 = Subtract(A=VectorParameter() {parameterName=Custom Location}.RGB, B=WorldPosition().XYZ)
  n2 = Normalize(VectorInput=@n1)
  NamedRerouteDeclaration(None=@n2) {name=DirectionToTarget, nodeColor=<rgba 0 0.7187 1 1>}
  n3 = Normalize(VectorInput=CrossProduct(A=NamedRerouteUsage(), B=Constant3Vector() {constant=<rgba 0 0 1 0>}))
  NamedRerouteDeclaration(None=@n3) {name=Right, nodeColor=<rgba 1 0.9609 0 1>}
  n4 = NamedRerouteUsage()
  n5 = VertexNormalWS()
  n6 = DotProduct(
   A=If(A=ScalarParameter() {parameterName=Facing Brightness Invert}, "A > B"=Multiply(B=@n4) {constA=-1}, "A < B"=@n4) {constB=0.5, equalsThreshold=0.001}
   B=@n5
  )
  NamedRerouteDeclaration(None=Power(Base=Saturate(None=@n6), Exp=ScalarParameter() {defaultValue=2, parameterName=Facing Brightness Falloff})) {name=Facing Fresnel mask, nodeColor=<rgba 0.4766 0 1 1>}
  n7 = Time()
  n8 = Multiply(A=@n7, B=Constant() {r=0.7123})
  n9 = Add(A=@n7, B=Multiply(A=Sine(None=@n8)) {constB=0.4})
  n10 = Cosine(None=Multiply(A=@n9, B=ScalarParameter() {defaultValue=7, parameterName=Flicker Speed Cosine}))
  n11 = Sine(None=Multiply(A=@n9, B=ScalarParameter() {defaultValue=2.1416, parameterName=Flicker Speed Sine}))
  n12 = Power(
   Base=Multiply(A=ConstantBiasScale(None=@n10), B=ConstantBiasScale(None=@n11))
   Exp=ScalarParameter() {defaultValue=4, parameterName=Flicker Speed Power}
  )
  NamedRerouteDeclaration(None=@n12) {name=velocity Phase, nodeColor=<rgba 0.5469 0 1 1>}
  n13 = TextureCoordinate()
  n14 = VectorParameter() {parameterName=CrackTiling, defaultValue=<rgba 1 1 0 1>}
  n15 = NamedRerouteDeclaration(None=AppendVector(A=@n14.R, B=@n14.R)) {name=CrackTiling, nodeColor=<rgba 0 0.4844 1 1>}
  n16 = Multiply(A=Multiply(A=@n13) {constB=2}, B=@n15)
  n17 = TextureObjectParameter() {
   parameterName=Normals
   samplerType=SAMPLERTYPE_Normal
   texture=/Hera/Props/SkyMango/SkyMango_Fragment_A/Textures/T_SkyMango_Fragment_Inner_S2shards_N
  }
  n18 = Multiply(A=@n15, B=@n13)
  n19 = TextureSample(UVs=@n18, Tex=@n17) {samplerType=SAMPLERTYPE_Normal, texture=/Hera/Props/SkyMango/SkyMango_Fragment_A/Textures/T_Glass_Cracks_N2}
  n20 = Add(
   A=TextureSample(UVs=@n16, Tex=@n17) {samplerType=SAMPLERTYPE_Normal, texture=&T_Glass_Cracks_N2}.RGB
   B=@n19.RGB
  )
  n21 = MaterialFunctionCall(
   Normal=StaticSwitchParameter(True=Multiply(A=@n20) {constB=0.5}, False=@n19.RGB) {parameterName=IsOuterShell}
   Flatness=ScalarParameter() {parameterName=Normal Flatness}
  ) {materialFunction=&FlattenNormal}
  NamedRerouteDeclaration(None=@n21.Result) {name=normals, nodeColor=<rgba 1 0.1641 0 1>}
  n22 = TextureObjectParameter() {
   parameterName=SpecularMasks
   samplerType=SAMPLERTYPE_Masks
   texture=/Hera/Props/SkyMango/SkyMango_Fragment_A/Textures/T_SkyMango_Fragment_Inner_S2shards_M
  }
  n23 = TextureSample(UVs=@n18, Tex=@n22) {texture=/Hera/Props/SkyMango/SkyMango_Fragment_A/Textures/T_Glass_Cracks_M}
  n24 = Add(A=TextureSample(UVs=@n16, Tex=@n22) {texture=&T_Glass_Cracks_M}.RGB, B=@n23.RGB)
  NamedRerouteDeclaration(None=StaticSwitchParameter(True=Multiply(A=@n24) {constB=0.5}, False=@n23.RGB) {parameterName=IsOuterShell}) {name=emissive, nodeColor=<rgba 0.8281 0 1 1>}
  NamedRerouteDeclaration(None=CrossProduct(A=@n3, B=NamedRerouteUsage())) {name=Up, nodeColor=<rgba 1 0.9609 0 1>}
  n25 = Multiply(
   A=DotProduct(A=NamedRerouteUsage(), B=WorldPosition().XYZ)
   B=Multiply(A=ScalarParameter() {defaultValue=3, parameterName=Streak Scale}) {constB=0.01}
  )
  n26 = Multiply(A=@n25) {constB=40}
  n27 = NamedRerouteUsage()
  n28 = Add(A=Multiply(A=@n25) {constB=17.3}, B=@n27)
  n29 = Multiply(A=Sine(None=Add(A=@n26, B=@n27)), B=Sine(None=Add(A=@n28) {constB=3.7}))
  NamedRerouteDeclaration(None=Power(Base=Saturate(None=@n29)) {constExponent=1}) {name=Streaks Mask, nodeColor=<rgba 0 1 0.3906 1>}
  n30 = DotProduct(A=NamedRerouteUsage(), B=WorldPosition().XYZ)
  n31 = Multiply(A=ScalarParameter() {defaultValue=3, parameterName=Streak Scale}) {constB=0.01}
  n32 = Time()
  n33 = ScalarParameter() {defaultValue=0.5, parameterName=Streak Speed}
  NamedRerouteDeclaration(None=Subtract(A=Multiply(A=@n30, B=@n31), B=Multiply(A=@n32, B=@n33))) {name=Streak V, nodeColor=<rgba 1 0.0703 0 1>}
  NamedRerouteDeclaration(None=@n25) {name=Streak U, nodeColor=<rgba 1 0.0703 0 1>}
  n34 = Time()
  n35 = ScalarParameter() {defaultValue=2, parameterName=Flicker Instance Lifetime}
  n36 = Divide(A=Fmod(A=@n34, B=@n35), B=@n35)
  NamedRerouteDeclaration(None=@n36) {name=Normalized Loop Age, nodeColor=<rgba 0.9219 1 0 1>}
  n37 = Length(None=@n1) {is=class MaterialExpressionLength}
  n38 = ScalarParameter() {defaultValue=45, parameterName=Flicker Attraction Max Distance}
  n39 = LinearInterpolate(B=Multiply(A=Clamp(None=@n37, Max=@n38), B=@n2), Alpha=@n12)
  NamedRerouteDeclaration(None=@n39) {name=velocity, nodeColor=<rgba 1 0.9844 0 1>}
  n40 = Sine(None=Multiply(A=@n34) {constB=3.7})
  NamedRerouteDeclaration(None=Add(A=Multiply(A=@n40) {constB=0.15}, B=@n36)) {name=Age Loop plus Jitter, nodeColor=<rgba 0.0547 0 1 1>}
  NamedRerouteUsage()
  NamedRerouteUsage()
  NamedRerouteUsage()
  NamedRerouteUsage()
  NamedRerouteUsage()
  n41 = NamedRerouteUsage()
  n42 = Power(Base=Saturate(None=DotProduct(A=@n4, B=@n5)), Exp=ScalarParameter() {defaultValue=2, parameterName=Facing Flicker Falloff})
  n43 = Time()
  n44 = Constant() {r=0.7314}
  n45 = Multiply(A=Sine(None=Multiply(A=@n43, B=@n44))) {constB=0.4}
  n46 = Constant() {r=8}
  n47 = Multiply(A=Multiply(A=ScalarParameter() {defaultValue=1.5, parameterName=Streak Flicker Speed}, B=Add(A=@n43, B=@n45)), B=Add(A=Frac(None=@n46)))
  n48 = Constant() {r=6.28}
  n49 = Multiply(
   A=MaterialFunctionCall(Seed=Constant() {r=53486}) {materialFunction=&MF_Hash}.Float
   B=Sine(None=Add(A=@n47, B=@n48))
  )
  n50 = Saturate(None=Subtract(A=Multiply(A=@n49) {constB=2}) {constB=0.5})
  n51 = Multiply(
   A=Multiply(A=Multiply(A=@n41, B=@n42), B=NamedRerouteDeclaration(None=@n50) {name=Streak Velocity, nodeColor=<rgba 1 0 0.2109 1>})
   B=ScalarParameter() {defaultValue=300, parameterName=Flicker Distance}
  )
  n52 = NamedRerouteUsage()
  n53 = ObjectPositionWS()
  n54 = WorldPosition()
  n55 = ScalarParameter() {defaultValue=-0.15, parameterName=Collapse Towards Custom Location Strength}
  n56 = NamedRerouteUsage()
  n57 = Add(A=@n39, B=Multiply(A=Subtract(A=@n53, B=@n54.XYZ), B=LinearInterpolate(A=@n55, B=@n55, Alpha=@n56) {constA=-1}))
  n58 = Add(A=Multiply(A=VertexColor().R, B=Multiply(A=@n51, B=@n52)), B=@n57)
  n59 = Constant()
  n60 = Normalize(VectorInput=Add(A=NamedRerouteUsage(), B=NamedRerouteUsage()))
  n61 = Multiply(A=@n60, B=ScalarParameter() {defaultValue=1500, parameterName=Motion Vector Scale})
  Reroute(None=Add(
    A=StaticSwitchParameter(True=@n58, False=@n57) {defaultValue=true, parameterName=Enable Streaks}
    B=PreviousFrameSwitch(Current Frame=@n59, Previous Frame=@n61)
   ))
  n62 = NamedRerouteUsage()
  n63 = ComponentMask(None=NamedRerouteUsage()) {r=true}
  n64 = ScalarParameter() {defaultValue=0.2, parameterName="Emissive Cells - SpecMasks (R)"}
  n65 = Add(
   A=Multiply(A=ComponentMask(None=@n62) {g=true}, B=ScalarParameter() {defaultValue=0.4, parameterName="Emissive Cracks - SpecMasks (G)"}) {constB=8}
   B=Add(A=Multiply(A=@n63, B=@n64), B=ScalarParameter() {parameterName="Emissive Cells Bias - SpecMasks (R)"})
  )
  n66 = VectorParameter() {parameterName=Emissive Color A, defaultValue=<rgba 2 0 0.1114 1>}
  n67 = VectorParameter() {parameterName=Emissive Color B, defaultValue=<rgba 0 0.2103 1 1>}
  n68 = NamedRerouteUsage()
  n69 = LinearInterpolate(
   A=@n66.RGB
   B=Multiply(A=ScalarParameter() {defaultValue=1.5, parameterName=Facing Brightness}, B=LinearInterpolate(A=@n66.RGB, B=@n67.RGB, Alpha=@n68))
   Alpha=NamedRerouteUsage()
  )
  Reroute(None=StaticSwitchParameter(True=@n69, False=Multiply(A=@n65, B=@n69)) {parameterName=IsStreaksMesh})
  ScalarParameter() {defaultValue=0.2, parameterName=Emissive Cracks Bias}
  n70 = Add(A=PixelDepth().R) {constB=-5}
  n71 = SceneTexture() {sceneTextureId=PPI_CustomDepth}
  n72 = Saturate(None=ComponentMask(None=Subtract(A=@n70, B=@n71.Color)) {r=true})
  n73 = OneMinus(None=@n60)
  n74 = DotProduct(A=Multiply(A=Transform(None=@n21.Result), B=VertexNormalWS()), B=NamedRerouteUsage())
  n75 = Clamp(
   None=Power(Base=Saturate(None=@n74), Exp=ScalarParameter() {defaultValue=3, parameterName=Fresnel Falloff Power})
   Min=ScalarParameter() {defaultValue=0.15, parameterName=Fresnel Min Opacity}
  )
  n76 = Power(
   Base=SphereMask(
    A=NamedRerouteUsage()
    B=ScalarParameter() {defaultValue=0.5, parameterName=Fade Jitter Age Threshold}
    Radius=ScalarParameter() {defaultValue=1, parameterName=Fade Jitter Radius}
    Hardness=ScalarParameter() {parameterName=Fade Jitter Hardness}
   )
   Exp=ScalarParameter() {defaultValue=1, parameterName=Fade Jitter Contrast}
  )
  n77 = Add(A=DotProduct(A=CameraVectorWS(), B=VertexNormalWS())) {constB=0.5}
  n78 = ScalarParameter() {defaultValue=5, parameterName=Streak Opacity Removal on Edge}
  n79 = Multiply(
   A=Multiply(A=Multiply(A=@n75, B=@n76), B=ScalarParameter() {defaultValue=3.5, parameterName=Opacity Multiplier})
   B=Saturate(None=Multiply(A=@n77, B=@n78))
  )
  n80 = Add(A=NamedRerouteUsage(), B=ScalarParameter() {defaultValue=0.3, parameterName=Velocity Phase Cutoff})
  n81 = StaticSwitchParameter(True=Multiply(A=Multiply(A=@n79, B=@n73), B=Ceil(None=@n80)), False=@n79) {parameterName=IsStreaksMesh}
  StaticSwitchParameter(True=@n81, False=Multiply(A=@n81, B=OneMinus(None=@n72))) {parameterName=IsStreaksMesh}
  Time()
 ]
}
